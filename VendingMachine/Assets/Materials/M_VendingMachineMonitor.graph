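# Unlit emissive material for the vending machine screen, showing a weapon icon, timer ring, resource icon and cost.
# It layers a radial countdown, rotating star, three dots and scanlines over a tinted vignette, with Slide offsetting the content.
# CostNumber becomes digits through per-column power-of-ten math into a number-sheet texture.
# Success adds rainbow color, Fail pulses fail colors, and Distort drives noise glitches that Low quality skips.
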Material /VendingMachine/Assets/Materials/M_VendingMachineMonitor
graph {
 properties {bUseMaterialAttributes=true, bUsedWithNanite=true, pixelDepthOffsetMode=PDOM_Legacy, shadingModel=MSM_Unlit, usedShadingModels=Unlit}
 nodes [
  MaterialFunctionCall() {materialFunction=&DebugTimeSine}
  MaterialFunctionCall() {materialFunction=&DebugTimeSine}
  MaterialFunctionCall() {materialFunction=&DebugTimeSine}
  MaterialFunctionCall() {materialFunction=&DebugTimeSine}
  n1 = Multiply(A=Time(), B=ScalarParameter() {defaultValue=22, parameterName=Block Speed A})
  n2 = TextureSample(UVs=Panner(Coordinate=TextureCoordinate() {uTiling=0.2}, Time=Ceil(None=@n1)) {bFractionalPart=true, speedX=0.7, speedY=-0.3}) {samplerType=SAMPLERTYPE_LinearColor, texture=/Game/Global/Textures/Default/Noise/T_Atlas_DigitalNoise_01}
  n3 = ScalarParameter() {parameterName=Distort}
  n4 = Multiply(A=@n2.RGB, B=@n3)
  n5 = Add(A=@n4)
  n6 = Divide(A=Subtract(A=Time(), B=ScalarParameter() {parameterName=TimerStartTime}), B=ScalarParameter() {defaultValue=8, parameterName=TimerDuration})
  n7 = Multiply(
   A=TextureCoordinate()
   B=MaterialFunctionCall(X=Constant() {r=1}, Y=ScalarParameter() {defaultValue=2, parameterName=HeightScale}) {materialFunction=&MakeFloat2}.Result
  )
  n8 = Multiply(A=LinearInterpolate(Alpha=@n2.B) {constA=-0.1, constB=0.1}, B=@n3)
  n9 = Multiply(A=Time()) {constB=0.2}
  n10 = Floor(None=Multiply(A=Frac(None=@n9)) {constB=1.1})
  n11 = TextureCoordinate() {vTiling=2.2}
  n12 = Multiply(A=Frac(None=ComponentMask(None=@n11) {g=true})) {constB=3}
  n13 = Multiply(A=Multiply(A=@n10, B=Floor(None=@n12)) {constA=0.3}, B=@n10)
  n14 = AppendVector(A=Clamp(None=Multiply(B=@n13) {constA=-0.03}) {minDefault=-1}, B=Constant())
  n15 = Add(
   A=QualitySwitch(Default=Add(A=@n8, B=@n14), Low=Constant())
   B=Multiply(A=ScalarParameter() {parameterName=Slide}, B=Constant2Vector() {r=1})
  )
  n16 = Constant() {r=-0.28}
  n17 = Saturate(None=Add(
    A=Add(A=@n7, B=@n15)
    B=MaterialFunctionCall(Y=@n16) {materialFunction=&MakeFloat2}.Result
   ))
  n18 = Add(A=@n17, B=Constant2Vector() {g=-0.045})
  n19 = MaterialFunctionCall(UVs=@n18, "Rotation Angle (0-1)"=ScalarParameter() {defaultValue=0.25, parameterName=Origin_0-1}) {materialFunction=&CustomRotator}
  n20 = Add(
   A=OneMinus(None=Clamp(None=@n6))
   B=MaterialFunctionCall(Vector or UVs=Subtract(A=@n19.Rotated Values) {constB=0.5}) {materialFunction=&VectorToRadialValue}.Vector Converted to Angle
  )
  n21 = Clamp(None=Multiply(A=ScalarParameter() {defaultValue=0.77, parameterName=OuterRadius_0-1}) {constB=0.5}) {maxDefault=0.5}
  n22 = Clamp(None=Multiply(A=MaterialFunctionCall!(UVs=@n18, CenterPosition, Radius=@n21) {materialFunction=/Engine/Functions/Engine_MaterialFunctions01/Gradient/SphereGradient-2D}.Result 0-1) {constB=2})
  n23 = Clamp(None=Multiply(A=ScalarParameter() {defaultValue=0.55, parameterName=InnerRadius_0-1}) {constB=0.5}) {maxDefault=0.495}
  n24 = Clamp(None=Multiply(A=MaterialFunctionCall(UVs=@n18, Radius=@n23) {materialFunction=&SphereGradient-2D}.Result 0-1) {constB=2})
  n25 = ComponentMask(None=Multiply(A=Floor(None=@n20), B=Subtract(A=@n22, B=@n24))) {r=true}
  n26 = AppendVector(A=@n25, B=@n25)
  n27 = Multiply(
   A=Clamp(None=AppendVector(A=@n26, B=@n25))
   B=Multiply(A=VectorParameter() {parameterName=ScreenColor, defaultValue=<rgba 0.056 0.3401 0.6979 0>}.RGB) {constB=0.3}
  ) {constA=1, constB=0.2}
  n28 = SphereMask(A=TextureCoordinate(), B=Constant() {r=0.5}, Radius=Constant() {r=0.8}, Hardness=Constant())
  n29 = ScalarParameter() {parameterName=Success}
  n30 = Add(A=Multiply(A=Time()) {constB=-0.5}, B=@n29)
  n31 = Multiply(A=TextureSample(UVs=Panner(Time=@n30) {speedY=-3}) {texture=/Game/Effects/Fort_Effects/Textures/Grad/T_Gradient_Rainbow_NoRed}.RGB) {constB=2.5}
  n32 = Add(A=VectorParameter() {parameterName=ScreenColor, defaultValue=<rgba 0.056 0.3401 0.6979 0>}.RGB, B=Multiply(A=@n31, B=@n29))
  n33 = Multiply(
   A=Add(A=TextureCoordinate(), B=@n15)
   B=MaterialFunctionCall(X=Constant() {r=1}, Y=ScalarParameter() {defaultValue=2, parameterName=HeightScale}) {materialFunction=&MakeFloat2}.Result
  )
  n34 = Constant2Vector() {g=-0.8}
  n35 = ComponentMask(None=Saturate(None=Add(A=@n33, B=@n34))) {g=true}
  n36 = Floor(None=Add(A=@n35) {constB=0.375})
  n37 = Floor(None=Add(A=@n35) {constB=0.125})
  n38 = Rotator(Coordinate=Add(A=Multiply(A=@n33) {constB=0.7}, B=Constant2Vector() {g=-0.08, r=0.15})) {speed=-1.5}
  n39 = Add(
   A=Multiply(A=Subtract(A=@n36, B=@n37), B=ScalarParameter() {parameterName=ShowResource})
   B=TextureSample(UVs=Saturate(None=@n38)) {texture=/Game/Athena/Items/Gameplay/VendingMachine/FX/T_StarShape}.R
  ) {constB=0}
  n40 = Multiply(A=@n28, B=LinearInterpolate(A=@n32, B=Multiply(A=@n32) {constB=0.45}, Alpha=Saturate(None=@n39)) {constAlpha=0})
  n41 = MaterialFunctionCall(Speed=Constant() {r=8}) {materialFunction=&DebugTimeSine}
  n42 = Multiply(
   A=Power(Base=OneMinus(None=@n28))
   B=Multiply(A=@n41.Result, B=VectorParameter() {parameterName=ScreenColorFail, defaultValue=<rgba 1 0.0573 0.0673 0>}.RGB)
  )
  n43 = Multiply(A=@n40) {constB=0.3}
  n44 = Multiply(
   A=TextureCoordinate()
   B=MaterialFunctionCall(X=Constant() {r=1}, Y=ScalarParameter() {defaultValue=2, parameterName=HeightScale}) {materialFunction=&MakeFloat2}.Result
  )
  n45 = Constant2Vector() {g=1.85, r=0.5}
  n46 = VectorParameter() {parameterName=CircleScales, defaultValue=<rgba 0.6 1.35 0.6 0>}
  n47 = Multiply(B=@n46.G) {constA=0.02}
  n48 = SphereMask(A=@n44, B=Constant2Vector() {g=1.85, r=0.4}, Radius=Multiply(B=@n46.R) {constA=0.02})
  n49 = SphereMask(A=@n44, B=Constant2Vector() {g=1.85, r=0.6}, Radius=Multiply(B=@n46.B) {constA=0.02})
  n50 = LinearInterpolate(
   A=LinearInterpolate(A=@n40, B=Add(A=@n42, B=@n43), Alpha=ScalarParameter() {parameterName=Fail})
   Alpha=Add(A=SphereMask(A=@n44, B=@n45, Radius=@n47), B=Add(A=@n48, B=@n49))
  )
  n51 = ScalarParameter() {parameterName=Success}
  n52 = TextureSampleParameter2D(UVs=@n17) {parameterName=WeaponIcon, texture=/Game/UI/Foundation/Textures/Icons/Weapons/Items/T-Icon-Weapons-SK-M-Med-Soldier-04-Minigun}
  n53 = MaterialFunctionCall(Hue Shift Percentage=@n52.RGB, Texture=@n31) {materialFunction=&HueShift}
  n54 = Add(A=@n52.RGB, B=Multiply(A=Multiply(A=@n51, B=@n53.Result) {constB=0.25}) {constB=0.2})
  n55 = Constant() {r=1}
  n56 = ScalarParameter() {defaultValue=2, parameterName=HeightScale}
  n57 = Constant() {r=3.5}
  n58 = Multiply(
   A=Multiply(
    A=TextureCoordinate()
    B=MaterialFunctionCall(X=@n55, Y=@n56) {materialFunction=&MakeFloat2}.Result
   )
   B=@n57
  )
  n59 = Constant2Vector() {g=-4.9, r=-0.35}
  n60 = TextureSampleParameter2D(UVs=Add(A=Add(A=@n58, B=@n59), B=Multiply(A=@n15, B=@n57))) {parameterName=ResourceIcon, samplerSource=SSM_Clamp_WorldGroupSettings, texture=/Game/UI/Icons/Icon-Resource/Icon-Resource-Wood-M}
  n61 = LinearInterpolate(
   A=LinearInterpolate(A=Add(A=@n27, B=@n50), B=QualitySwitch(Default=@n54, Low=@n52.RGB), Alpha=@n52.A)
   B=@n60.RGB
   Alpha=Multiply(A=@n60.A, B=ScalarParameter() {parameterName=ShowResource})
  )
  n62 = QualitySwitch(Default=Add(A=Multiply(A=@n61, B=@n5), B=Multiply(A=@n4) {constB=0.3}), Low=@n61)
  n63 = VectorParameter() {parameterName=TextColor, defaultValue=<rgba 1 1 1 0>}
  n64 = LinearInterpolate(
   A=@n63.RGB
   B=LinearInterpolate(
    A=Multiply(A=@n63.RGB) {constB=0.15}
    B=VectorParameter() {parameterName=TextColorFail, defaultValue=<rgba 5 0.051 0.1505 0>}.RGB
    Alpha=Round(None=@n41.Result)
   )
   Alpha=Ceil(None=ScalarParameter() {parameterName=Fail})
  ) {constAlpha=1}
  n65 = Constant2Vector() {g=0.72, r=0.4}
  n66 = MaterialFunctionCall!(
   UVs=Add(A=TextureCoordinate(), B=@n15)
   TextureIn
   UpperLeftCorner=Add(A=Constant2Vector(), B=@n65)
   LowerRightCorner=Add(A=Constant2Vector() {g=0.11, r=0.35}, B=@n65)
   ExportFloat4
  ) {materialFunction=/Engine/Functions/Engine_MaterialFunctions01/Texturing/TextureCropping}
  n67 = Multiply(A=@n66.Crop UVs, B=Constant2Vector() {g=0.1, r=1}) {constA=2, constB=0.1}
  n68 = Multiply(A=OneMinus(None=ComponentMask(None=@n67) {r=true}), B=Constant() {r=3}) {constB=10}
  n69 = ComponentMask(None=@n67) {g=true}
  n70 = Constant() {r=10}
  n71 = Add(A=Ceil(None=Max(A=@n68) {constB=0}))
  n72 = Frac(None=Divide(A=ScalarParameter() {defaultValue=250, parameterName=CostNumber}, B=Power(Base=@n70, Exp=@n71) {desc="10-100,000,000"}))
  n73 = Divide(A=Floor(None=Multiply(A=@n72) {constB=100})) {constB=10}
  n74 = TextureSample!(UVs=AppendVector(A=Frac(None=@n68), B=Add(A=@n69, B=@n73)), Tex, Level, Apply View MipBias) {mipValueMode=TMVM_MipLevel, samplerType=SAMPLERTYPE_Masks, texture=/Game/Athena/Items/Gameplay/VendingMachine/FX/T_NumberSheet}
  n75 = Multiply(A=Saturate(None=Multiply(A=@n66.Crop Mask, B=@n74.R) {desc=crop mask}), B=ScalarParameter() {parameterName=ShowResource})
  n76 = ScalarParameter() {defaultValue=1, parameterName=ScreenBrightness}
  n77 = Multiply(
   A=TextureCoordinate() {vTiling=32}
   B=MaterialFunctionCall(X=Constant() {r=1}, Y=ScalarParameter() {defaultValue=2, parameterName=HeightScale}) {materialFunction=&MakeFloat2}.Result
  )
  n78 = Multiply(A=TextureSample!(UVs=Panner(Coordinate=@n77) {speedX=0.5, speedY=0.7}, Tex, Apply View MipBias) {texture=/Game/Missions/Primary/LaunchTheBalloon/Textures/T_TV_Scanlines}.RGB) {constB=0.04}
  n79 = Multiply(
   A=Add(A=LinearInterpolate(A=@n62, B=@n64, Alpha=@n75), B=Multiply(A=@n76, B=@n78))
   B=ScalarParameter() {defaultValue=1, parameterName=ScaleEmissiveBrightness}
  )
  MaterialFunctionCall(MA In=SetMaterialAttributes(Emissive Color=MaterialFunctionCall(Emissive=@n79) {materialFunction=&MF_FX_EmissiveAdjustments}.Result) {attributeSetTypes [#1]}) {materialFunction=&MF_CreativeMoveTool}
 ]
}
